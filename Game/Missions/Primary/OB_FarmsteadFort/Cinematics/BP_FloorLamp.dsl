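; Damageable floor lamp for the Farmstead Fort mission whose two spotlights switch off when it is nearly or fully destroyed.
; Server damage at 5% health or below, or actor destruction, triggers a DoOnce-guarded call that hides both spotlights and sets a disable flag.
; The flag's OnRep repeats the hiding on clients.
; The construction script applies exposed intensity, cone and radius to the first spotlight and color to both.
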(event EventDestroyed
  (Utilities|FlowControl|DoOnce false
    (:Completed
      (CallFunction|DisabaleLight))))

(event EventOnDamageServer
  (params (Damage :type Float%20%28single-precision%29) (DamageTags :type Gameplay%20Tag%20Container%20Structure%20%28by%20ref%29) (Momentum :type Vector :default "0, 0, 0") (HitInfo :type Hit%20Result%20Structure%20%28by%20ref%29) (InstigatedBy :type Controller%20Object%20Reference) (DamageCauser :type Actor%20Object%20Reference) (EffectContext :type Gameplay%20Effect%20Context%20Handle%20Structure))
  (if (<= (Health|GetHealthPercent self) 0.05)
    (Utilities|FlowControl|DoOnce false
      (:Completed
        (CallFunction|DisabaleLight)))))

(event Custom|disabale_light
  (bind _spotlight (Variables|Default|GetSpotLight))
  (bind _spotlight1 (Variables|Default|GetSpotLight1))
  (Rendering|SetVisibility _spotlight false false)
  (Development|SetHiddeninGame _spotlight true false)
  (Rendering|SetVisibility _spotlight1 false false)
  (Development|SetHiddeninGame _spotlight1 true false)
  (Variables|Default|Setdisable%20light true))

(fn OnRep_disablelight
  (bind _spotlight (Variables|Default|GetSpotLight))
  (bind _spotlight1 (Variables|Default|GetSpotLight1))
  (if (Variables|Default|Getdisable%20light)
    (Development|SetHiddeninGame _spotlight true false)
    (Rendering|SetVisibility _spotlight false false)
    (Rendering|SetVisibility _spotlight1 false false)
    (Development|SetHiddeninGame _spotlight1 true false)))

(fn ConstructionScript
  (bind _spotlight (Variables|Default|GetSpotLight))
  (bind _lightcolor (Variables|Default|GetLightColor))
  (Rendering|Components|Light|SetIntensity _spotlight (Variables|Default|GetLightIntensity))
  (Rendering|Lighting|SetOuterConeAngle _spotlight (Variables|Default|GetLightOuterConeValue))
  (Rendering|Lighting|SetAttenuationRadius _spotlight (Variables|Default|GetLightAttenuationRadius))
  (Rendering|Components|Light|SetLightColor (Variables|Default|GetSpotLight) _lightcolor true)
  (Rendering|Components|Light|SetLightColor (Variables|Default|GetSpotLight1) _lightcolor true))
